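# Niagara sprite effect for the TubaGummi skin that drips slowly from eight arm, hand and calf bones.
# Each drip stays pinned to its bone while stretching, then lerps into free fall around mid-life, aligning to velocity affected by gravity and owner movement.
# Color follows a time-of-day curve with a per-group tint.
# Hand-hidden and foot-hidden user bools set a per-group visibility tag through lookup arrays.
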NiagaraSystem /BRCosmetics/Characters/Player/Male/Medium/Bodies/M_MED_Gummi/Skins/TubaGummi/FX/NS_TubaGummi_Drips
dependencies {
 usedDataInterfaces [
  &NiagaraDataInterfaceColorCurve
  &NiagaraDataInterfaceArrayFloat3
  &NiagaraDataInterfaceCurve
  &NiagaraDataInterfaceSkeletalMesh
  &NiagaraDataInterfaceVector2DCurve
  &NiagaraDataInterfaceArrayInt32
 ]
 usedDynamicInputs []
 usedModules [&SystemState]
}
emitters {Drips {
  emitterUpdateScript [
   &EmitterState()
   &SpawnRate(SpawnRate=0.7, Use Spawn Probability=false)
  ]
  particleSpawnScript [
   &FN_SetSpawnGroup(Number of Spawn Groups=8)
   &InitializeParticle(
    Color=&MultiplyLinearColor(
     Linear Color=&ColorFromCurve(
      ColorCurve=DataInterface() {propertyValues {
        AlphaCurve {keys [<at 0 1>,<at 0.125 1>,<at 0.25 1>,<at 0.375 1>,<at 0.5 1>,<at 0.625 1>,<at 0.75 1>,<at 0.875 1>,<at 1 1>]}
        BlueCurve {keys [<at 0 0.4792>,<at 0.208 0.5395>,<at 0.25 0.6172>,<at 0.291 0.6939>,<at 0.333 0.7379>,<at 0.375 0.7835>,<at 0.416 0.7991>,<at 0.458 0.8148>,<at 0.5 0.8308>,<at 0.541 0.8148>,<at 0.583 0.7991>,<at 0.625 0.7991>,<at 0.666 0.7529>,<at 0.708 0.7084>,<at 0.75 0.5711>,<at 0.791 0.4735>,<at 0.833 0.3231>,<at 1 0.4792>]}
        GreenCurve {keys [<at 0 0.4792>,<at 0.208 0.5271>,<at 0.25 0.6376>,<at 0.291 0.7529>,<at 0.333 0.7835>,<at 0.375 0.8148>,<at 0.416 0.7991>,<at 0.458 0.8148>,<at 0.5 0.8469>,<at 0.541 0.7991>,<at 0.583 0.7991>,<at 0.625 0.8148>,<at 0.666 0.7991>,<at 0.708 0.7529>,<at 0.75 0.5711>,<at 0.791 0.5029>,<at 0.833 0.5029>,<at 1 0.4792>]}
        RedCurve {keys [<at 0 0.4792>,<at 0.208 0.6867>,<at 0.25 0.8148>,<at 0.291 0.8879>,<at 0.333 0.8963>,<at 0.375 0.9131>,<at 0.416 0.8714>,<at 0.458 0.8714>,<at 0.5 0.9131>,<at 0.541 0.8632>,<at 0.583 0.8714>,<at 0.625 0.9131>,<at 0.666 0.8963>,<at 0.708 0.8714>,<at 0.75 0.7529>,<at 0.791 0.7529>,<at 0.833 0.8469>,<at 1 0.4792>]}
       }}
      CurveIndex=Linked() {linkedVariable {name=NPC.FortniteNPC.TimeOfDay_0-1, type=NiagaraFloat}}
     )
     Scale Factor=&MakeVector4_FromLinearColor(LinearColor=&MakeCustomLinearColorFromBool(
       Bool=&SetBoolByIntComparison(A=$Particles.SpawnGroup, B=6, Comparison Type=::NewEnumerator2)
       True Color=<rgba 0.904 0.9799 1 1>
       False Color=<rgba 1 1 1 1>
      ))
    )
    Lifetime Max=1.8
    Lifetime Min=1.2
    Lifetime Mode=::NewEnumerator1
    Sprite Size=<2.5 2.5>
    Sprite Size Mode=::NewEnumerator4
    UsePositionOffset=false
    Write Material Random=true
   )
   /SetVariables(
    Particles.PosOffset:Vector3f=&SelectVectorFromArray(
     Direct Array Index=$Particles.SpawnGroup
     Vector Selection Array=DataInterface() {propertyValues {MaxElements=8, FloatData [<0 0 0>, <0 0 0>, <1 1 0>, <-3 -2 0>, <0 0 0>, <0 0 0>, <0 0 0>, <0 0 0>]}}
    )
    Particles.SpriteAlignment=<0 0 -1>
   ) {isSetParameters=true}
  ]
  particleUpdateScript [
   &ParticleState()
   &GravityForce(Gravity=&MakeVector(
     X=0
     Y=0
     Z=&FloatFromCurve(FloatCurve=DataInterface() {propertyValues {Curve {keys [<at 0 0>, {arriveTangent=-2463.0549, leaveTangent=-2463.0549, time=0.4, value=0}, <at 0.6 -700>]}}})
    ))
   &SkeletalMeshLocation(
    Bone / Socket Index=$Particles.SpawnGroup
    Bone Sampling Mode=::NewEnumerator4
    MaskSpawnGroup=false
    Orientation Sampling=::NewEnumerator1
    Position Sampling=::NewEnumerator2
    Sampled Position Offset=$Particles.PosOffset
    Skeletal Mesh=DataInterface() {propertyValues {
      DefaultMesh=null
      ExcludeBoneName=root
      OptionalComponentClass=null
      PreviewMesh=/BRCosmetics/Characters/Player/Male/Medium/Bodies/M_MED_Gummi/Meshes/M_MED_Gummi
      SkinningMode=SkinOnTheFly
      SoftSourceActor=null
      SourceMode=Default
      UvSetIndex=0
      WholeMeshLOD=-1
      bExcludeBone=false
      bReadDeformedGeometry=true
      bRequireCurrentFrameData=false
      ComponentTags []
      FilteredBones [upperarm_l, upperarm_r, lowerarm_l, lowerarm_r, dyn_calfMeasure_r, dyn_calfMeasure_l, hand_l, hand_r]
      FilteredSockets []
      MeshUserParameter {parameter {name=None, typeDefHandle {registeredTypeIndex=1}, varData []}}
      SamplingRegions []
     }}
   )
   &LinearForce(
    Coordinate Space:ENiagaraCoordinateSpace=::World
    Force=&Multiply_Vector(
     A=&MakeCustomVectorFromBool(
      Bool=&SetBoolByFloatComparison(
       A=&MakeFloatFromVector(Vector=$Engine.Owner.Velocity, Channel=::NewEnumerator2)
       B=-50
       Comparison Type=::NewEnumerator6
      )
      True Vector=$Engine.Owner.Velocity
      False Vector=&Multiply_Vector(A=$Engine.Owner.Velocity, B=<1 1 0>)
     )
     B=<-0.4 -0.4 6>
    )
   )
   &SolveForcesAndVelocity(Clamp Velocity=true, Limit Acceleration=false)
   &ScaleSpriteSize(
    Scale Factor=&MakeVector2D(
     X=&FloatFromCurve(FloatCurve=DataInterface() {propertyValues {Curve {keys [
          {arriveTangent=-0.1045, interpMode=RCIM_Cubic, leaveTangent=-0.1045, tangentMode=RCTM_User, time=0, value=2.5}
          {arriveTangent=-1.7, interpMode=RCIM_Cubic, leaveTangent=-1.7, time=0.6, value=1}
          {interpMode=RCIM_Cubic, leaveTangent=0.625, time=1, value=0.8}
         ]}}})
     Y=&FloatFromCurve(FloatCurve=DataInterface() {propertyValues {Curve {keys [
          <at 0 0 RCIM_Cubic>
          {arriveTangent=6.6667, interpMode=RCIM_Cubic, leaveTangent=6.6667, time=0.5, value=4}
          {arriveTangent=25.2206, interpMode=RCIM_Cubic, leaveTangent=25.2206, tangentMode=RCTM_User, time=0.75, value=5}
          <at 1 5 RCIM_Cubic>
         ]}}})
    )
    Scale Sprite Size Mode=::NewEnumerator2
   )
   /SetVariables(Particles.Position=&LerpPosition(
     A=$Particles.SkeletalMeshLocation.SampledPosition
     B=$Particles.Position
     Alpha=&FloatFromCurve(FloatCurve=DataInterface() {propertyValues {Curve {keys [<at 0.5 0>,<at 0.6 1>]}}})
    )) {isSetParameters=true}
   /SetVariables(Particles.SpriteAlignment=&Lerp_Vector(
     A=<0 0 -1>
     B=&Multiply_Vector(A=$Particles.Velocity, B=<2 2 1>)
     Alpha=&FloatFromCurve(FloatCurve=DataInterface() {propertyValues {Curve {keys [<at 0.5 0>,<at 0.6 1>]}}})
    )) {isSetParameters=true}
   /SetVariables(Particles.VisibilityTag=&MakeCustomIntFromBool(
     Bool=$User.bisHandHidden
     True INT=&MakeCustomIntFromBool(
      Bool=$User.bisFootHidden
      True INT=&SelectIntFromArray(Direct Array Index=$Particles.SpawnGroup, Int Selection Array=DataInterface() {propertyValues {IntData [0,0,1,1,1,1,1,1]}})
      False INT=&SelectIntFromArray(Direct Array Index=$Particles.SpawnGroup, Int Selection Array=DataInterface() {propertyValues {IntData [0,0,1,1,0,0,1,1]}})
     )
     False INT=&MakeCustomIntFromBool(
      Bool=$User.bisFootHidden
      True INT=&SelectIntFromArray(Direct Array Index=$Particles.SpawnGroup, Int Selection Array=DataInterface() {propertyValues {IntData [0,0,0,0,1,1,0,0]}})
      False INT=0
     )
    )) {isSetParameters=true}
  ]
  rendererClasses [&NiagaraSpriteRendererProperties]
  renderers [{rendererClass=&NiagaraSpriteRendererProperties}]
 }}
summary {
 systemName=NS_TubaGummi_Drips
 emitters [{emitterName=Drips, rendererClasses [&NiagaraSpriteRendererProperties]}]
 userVariables [
  User.bisFootHidden {type=NiagaraBool}
  User.bisHandHidden {type=NiagaraBool}
 ]
}
